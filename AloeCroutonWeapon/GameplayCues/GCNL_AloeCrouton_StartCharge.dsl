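; Looping gameplay cue for the start of the AloeCrouton weapon charge, handling its sound and camera shake.
; On application it attaches the charge sound to the player pawn, sets its IsLocal parameter from whether that pawn is locally viewed, and starts the controller's camera shake.
; Removal stops the shake and fades out the sound.
; Sound indicator overrides use Neutral affiliation and show only to players not on the instigator's team.
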(event EventBeginPlay)

(event Collision|EventActorBeginOverlap
  (params (OtherActor :type Actor%20Object%20Reference)))

(event EventTick
  (params (DeltaSeconds :type Float%20%28single-precision%29)))

(event EventOnApplication(EitherVFX-preferthisone)
  (params (MyTarget :type Actor%20Object%20Reference) (Parameters :type Gameplay%20Cue%20Parameters%20Structure%20%28by%20ref%29) (ParticleComponents :type Array%20of%20FXSystem%20Component%20Object%20References) (AudioComponents :type Array%20of%20Audio%20Component%20Object%20References) (BurstCameraShakeInstance :type Camera%20Shake%20Base%20Object%20Reference) (BurstDecalInstance :type Decal%20Actor%20Object%20Reference))
  ;[1 1 1] Audio logic and is local parameter failsafe
  (bind _asfort_player_pawn_athena (Utilities|Casting|CastToFortPlayerPawnAthena MyTarget)
    (:then
      (bind _returnvalue (Audio|SpawnSoundAttached :Sound (Variables|Default|GetChargeStartSound) :AttachToComponent (Variables|/Script/Engine.Actor|GetRootComponent _asfort_player_pawn_athena) :AttachPointName "None" :LocationType "KeepRelativeOffset" :bStopWhenAttachedToDestroyed true :VolumeMultiplier 1.0 :PitchMultiplier 1.0 :StartTime 0.0 :bAutoDestroy true))
      (bind _output_get (Variables|Default|SetChargeStartAudioComponent _returnvalue))
      (Audio|Parameter|SetBooleanParameter _output_get "IsLocal" (Pawn|IsLocallyViewed _asfort_player_pawn_athena) :!class /Script/AudioExtensions.AudioParameterControllerInterface))
    (:CastFailed
      _))
  (Variables|Default|SetParameters Parameters)
  (Game|Feedback|ClientStartCameraShake :self (Fortnite|GetFortPlayerControllerfromActor MyTarget) :Shake (Variables|Default|GetCameraShake) :Scale 1.0 :PlaySpace "CameraLocal"))

(event EventOnLoopingStart(EitherVFX-preferthisversion)
  (params (MyTarget :type Actor%20Object%20Reference) (Parameters :type Gameplay%20Cue%20Parameters%20Structure%20%28by%20ref%29) (ParticleComponents :type Array%20of%20FXSystem%20Component%20Object%20References) (AudioComponents :type Array%20of%20Audio%20Component%20Object%20References))
  (Variables|Default|SetLoopingAudioComponents AudioComponents)
  (Variables|Default|SetParticle%20Components ParticleComponents))

(event EventOnRemoval(EitherVFX-preferthisversion)
  (params (MyTarget :type Actor%20Object%20Reference) (Parameters :type Gameplay%20Cue%20Parameters%20Structure%20%28by%20ref%29) (ParticleComponents :type Array%20of%20FXSystem%20Component%20Object%20References) (AudioComponents :type Array%20of%20Audio%20Component%20Object%20References) (BurstCameraShakeInstance :type Camera%20Shake%20Base%20Object%20Reference) (BurstDecalInstance :type Decal%20Actor%20Object%20Reference))
  (Game|Feedback|ClientStopCameraShake (Fortnite|GetFortPlayerControllerfromActor MyTarget) (Variables|Default|GetCameraShake) true)
  (bind _chargestartaudiocomponent (Variables|Default|GetChargeStartAudioComponent)
    (:then
      (Audio|Components|Audio|FadeOut _chargestartaudiocomponent 0.5 0.0 "Linear" :!class /Script/Engine.AudioComponent))
    (:else)))

(fn GetAffiliationsToShowFor
  (return (Utilities|Array|MakeArray "Neutral")))

(fn ShouldShowSoundIndicator
  (params (PlayerController :type Fort%20Player%20Controller%20Object%20Reference))
  (return (not (Team|OnSameTeam PlayerController (Ability|GameplayCue|GetInstigatorActor (Variables|Default|GetParameters))))))

(fn ConstructionScript)
